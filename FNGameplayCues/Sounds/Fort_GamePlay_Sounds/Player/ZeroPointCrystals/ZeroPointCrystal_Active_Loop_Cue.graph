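# Looping Sound Cue for an active Zero Point Crystal, classed under gadget and item audio.
# A looping wave feeds an Enveloper that fades volume in over about half a second, then a Modulator at default settings.
# It uses a dedicated active-loop attenuation asset, a reduced cue volume, and overridden concurrency that caps simultaneous instances.
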SoundCue /FNGameplayCues/Sounds/Fort_GamePlay_Sounds/Player/ZeroPointCrystals/ZeroPointCrystal_Active_Loop_Cue
root=Modulator_2
nodes [
 {
  id=Enveloper_0
  children [WavePlayer_8]
  properties {pitchCurve {editorCurveData {keys [{time=0}]}}, volumeCurve {editorCurveData {keys [<at 0.0172 0.0116>,<at 0.5444 1>]}}}
 }
 {id=Modulator_2, children [Enveloper_0]}
 {
  id=WavePlayer_8
  properties {bLooping=true, soundWaveAssetPtr=/Game/Sounds/Fort_GamePlay_Sounds/Player/ZeroPointCrystals/ZeroPointCrystal_Active_Loop_01}
 }
]
properties {
 attenuationSettings=/FNGameplayCues/Sounds/Attenuation/Fort_ZeroPointCrystal_ActiveLoop_Att
 bOverrideConcurrency=true
 maxDistance=6000
 soundClassObject=&Fort_Gadgets_Items
 volumeMultiplier=0.3
 concurrencyOverrides {maxCount=3}
}
