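# Sound cue that plays the power-down sound when a Zero Point Crystal deactivates.
# A single wave player feeds a modulator at the root, with the cue volume turned well down.
# It uses the crystal's active-loop attenuation asset and routes to the gadgets and items sound class.
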SoundCue /FNGameplayCues/Sounds/Fort_GamePlay_Sounds/Player/ZeroPointCrystals/ZeroPointCrystal_Deactivate_Cue
root=Modulator_1
nodes [
 {id=Modulator_1, children [WavePlayer_0]}
 {
  id=WavePlayer_0
  properties {soundWaveAssetPtr=&/ZeroPointCrystal_PowerDown_01}
 }
]
properties {
 attenuationSettings=/FNGameplayCues/Sounds/Attenuation/Fort_ZeroPointCrystal_ActiveLoop_Att
 maxDistance=6000
 soundClassObject=&Fort_Gadgets_Items
 volumeMultiplier=0.2
}
