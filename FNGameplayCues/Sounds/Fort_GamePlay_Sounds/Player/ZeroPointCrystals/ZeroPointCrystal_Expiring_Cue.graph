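# Sound cue for a Zero Point Crystal expiring, classed as a gadget item sound with active-loop attenuation.
# After a short delay, a Modulator feeds on a Mixer of two layers.
# The expire loop wave runs through an Enveloper that fades volume in and out while pitch falls.
# A quieter power-down wave starts after a further delay, landing during the loop layer's envelope.
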SoundCue /FNGameplayCues/Sounds/Fort_GamePlay_Sounds/Player/ZeroPointCrystals/ZeroPointCrystal_Expiring_Cue
root=Delay_1
nodes [
 {id=Delay_0, children [WavePlayer_1], properties {delayMax=1.75, delayMin=1.75}}
 {id=Delay_1, children [Modulator_2], properties {delayMax=1, delayMin=1}}
 {
  id=Enveloper_0
  children [WavePlayer_0]
  properties {
   pitchCurve {editorCurveData {keys [<at 0 1.5014>,<at 4.7708 1.0359>]}}
   volumeCurve {editorCurveData {keys [<at 0.0172 0.0116>,<at 0.9714 1.0053>,<at 2.5909 1.007>,<at 4.1723 -0.0199>]}}
  }
 }
 {id=Mixer_0, children [Enveloper_0,Delay_0], properties {inputVolume [1,0.2]}}
 {id=Modulator_2, children [Mixer_0]}
 {
  id=WavePlayer_0
  properties {soundWaveAssetPtr=&/ZeroPointCrystal_Expire_Loop_01}
 }
 {
  id=WavePlayer_1
  properties {soundWaveAssetPtr=&/ZeroPointCrystal_PowerDown_02}
 }
]
properties {
 attenuationSettings=/FNGameplayCues/Sounds/Attenuation/Fort_ZeroPointCrystal_ActiveLoop_Att
 maxDistance=6000
 soundClassObject=&Fort_Gadgets_Items
 volumeMultiplier=0.6
}
