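# Sound cue for the Zero Point Crystal phase jump, classed under gadget and item sounds.
# An equal-weight random node picks one of five shortened phase-jump waves, and a modulator varies its pitch and volume on each play.
# It is spatialized with the Zero Point Crystal attenuation and shares a concurrency setting from the Low Gravity Stone jump/land sounds.
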SoundCue /FNGameplayCues/Sounds/Fort_GamePlay_Sounds/Player/ZeroPointCrystals/ZeroPointCrystal_PhaseJump_Cue
root=Modulator_0
nodes [
 {id=Modulator_0, children [Random_1], properties {pitchMax=1.1, pitchMin=0.9, volumeMax=1, volumeMin=0.7}}
 {id=Random_1, children [WavePlayer_0,WavePlayer_1,WavePlayer_2,WavePlayer_8,WavePlayer_9], properties {weights [1,1,1,1,1]}}
 {
  id=WavePlayer_0
  properties {soundWaveAssetPtr=&/ZeroPointCrystal_PhaseJump_01_Shortened}
 }
 {
  id=WavePlayer_1
  properties {soundWaveAssetPtr=&/ZeroPointCrystal_PhaseJump_02_Shortened}
 }
 {
  id=WavePlayer_2
  properties {soundWaveAssetPtr=&/ZeroPointCrystal_PhaseJump_03_Shortened}
 }
 {
  id=WavePlayer_8
  properties {soundWaveAssetPtr=&/ZeroPointCrystal_PhaseJump_04_Shortened}
 }
 {
  id=WavePlayer_9
  properties {soundWaveAssetPtr=&/ZeroPointCrystal_PhaseJump_05_Shortened}
 }
]
properties {
 attenuationSettings=/Game/Sounds/Attenuation/Fort_ZeroPointCrystal_Att
 maxDistance=6500
 soundClassObject=&Fort_Gadgets_Items
 volumeMultiplier=1.25
 concurrencySet [/Game/Sounds/Fort_GamePlay_Sounds/Player/LowGravityStones/LowGravityStone_JumpLand_Concc]
}
